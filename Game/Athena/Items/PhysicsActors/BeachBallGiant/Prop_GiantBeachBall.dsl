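; Giant beach ball physics prop that players kick or shoot into the air, counting bounces and reporting kicks to a quest.
; Tagged-actor hits and non-fatal damage, which refills health, feed cooldown-gated BounceAndCount, applying scalable-float impulses with optional client prediction; pickaxe impulses follow the attacker-to-impact direction.
; Other tagged-interface hits end an active count, resetting immediately or after a delay once it reaches display threshold.
; Niagara effects mark impacts and strong bounces.
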(event EventOnDamageServer
  (params (Damage :type Float%20%28single-precision%29) (DamageTags :type Gameplay%20Tag%20Container%20Structure%20%28by%20ref%29) (Momentum :type Vector :default "0, 0, 0") (HitInfo :type Hit%20Result%20Structure%20%28by%20ref%29) (InstigatedBy :type Controller%20Object%20Reference) (DamageCauser :type Actor%20Object%20Reference) (EffectContext :type Gameplay%20Effect%20Context%20Handle%20Structure))
  ;[1 1 1] Damage
  (CallFunction|HandleCustomImpulsefromDamageInternal Momentum Damage DamageTags HitInfo InstigatedBy DamageCauser EffectContext))

(event Collision|EventHit
  (params (MyComp :type Primitive%20Component%20Object%20Reference) (Other :type Actor%20Object%20Reference) (OtherComp :type Primitive%20Component%20Object%20Reference) (bSelfMoved :type Boolean) (HitLocation :type Vector :default "0, 0, 0") (HitNormal :type Vector :default "0, 0, 0") (NormalImpulse :type Vector :default "0, 0, 0") (Hit :type Hit%20Result%20Structure%20%28by%20ref%29))
  (bind (_returnvalue_x _returnvalue_y) (Math|Vector|BreakVector (* (ImpulseGetters/Setters|GetHorizontalHitImpulse) HitNormal)))
  ;[1 1 1] On hit
  (|CallParentFunction MyComp Other OtherComp bSelfMoved HitLocation HitNormal NormalImpulse Hit)
  (Utilities|FlowControl|Sequence
    (:then_0
      (if (not (Variables|Default|GetbUseClientPredictionOnImpulse))
        (Networking|SwitchHasAuthority
          (:Authority
            _)))
      (if (Utilities|DoesObjectImplementInterface Other "/Script/GameplayTags.GameplayTagAssetInterface")
        (bind _returnvalue (GameplayTags|HasAnyMatchingGameplayTags(Message) Other (Variables|Default|GetHitBounceTags)))
        (if _returnvalue
          (bind _result (CallFunction|CanBounceNow))
          (if _result
            (CallFunction|BounceandCount (Math|Vector|MakeVector _returnvalue_x _returnvalue_y (ImpulseGetters/Setters|GetVerticalHitImpulse)) (Transformation|GetActorLocation Other) true)
            (CallFunction|ClientPlayerKickSFX Other)
            ;;[1 1 1] Quest stat event
            (Quest|SendComplexCustomStatEvent (Quests|GetQuestManager (Fortnite|GetFortPlayerControllerfromActor Other) "Athena") 0 "(GameplayTags=((TagName=\"Athena.Quests.KickGiantBeachball\")))" "(GameplayTags=)" 1))
          (elif (Variables|Default|GetbounceCountActive)
            (if (>= (Variables|Default|GetbounceCount) (Variables|Default|GetminBouncesToDisplay))
              (Variables|Default|SetbounceCountActive false)
              (Variables|Default|SetisResetting true)
              (Variables|Default|SetstoppedBouncingTime (Utilities|Time|GetGameTimeinSeconds))
              (Utilities|Time|SetTimerbyEvent (delegate Custom|ResetBounces) (ScalableFloat|GetValueatLevel (Variables|Default|GetResetWaitTime)))
              (else
                (CallFunction|ResetBounces) ; Calling reset event to allow for child overrides.
              ))))
        (else
          _)))
    (:then_1
      (if (< (Math|Vector|VectorLength NormalImpulse) 3.5e+02)
        _
        (elif (Networking|IsServer)
          _
          (else
            ;[1 1 1] BounceFX: only if strong enough impulse
            (bind _returnvalue_2 (Niagara|SpawnSystematLocation "/Game/Athena/Items/PhysicsActors/BeachBallGiant/FX/NS_Physics_BeachBallGiant_Bounce.NS_Physics_BeachBallGiant_Bounce" HitLocation (Math|Conversions|RotationFromXVector HitNormal)))
            (Niagara|SetNiagaraVariable(Vector3) _returnvalue_2 "User.PhysicsVelocity" (Transformation|GetComponentVelocity (Variables|Debug|GetStaticMeshComponent)))))))))

(event Custom|ResetBounces
  ;[1 1 1] On hit
  (Development|PrintString "Reset")
  (Variables|Default|SetbounceCount -1)
  (Variables|Default|SetbounceCountActive false)
  (Variables|Default|SetisResetting false)
  (Variables|Default|SetstartedBouncingTime 0.0)
  (Variables|Default|SetstoppedBouncingTime 0.0))

(event Custom|ClientPlayerKickSFX
  (params (InteractingObject :type Object%20Reference))
  ;[1 1 1] SFX for players kicking the ball
  (Networking|CanExecuteCosmeticEvents
    (:True
      (Utilities|Casting|CastToFortPawn InteractingObject)
      (Audio|SpawnSoundAttached (Variables|Default|GetSFX_PlayerKick) (Variables|Debug|GetStaticMeshComponent)))))

(event EventOnDamagePlayEffects
  (params (Damage :type Float%20%28single-precision%29) (DamageTags :type Gameplay%20Tag%20Container%20Structure%20%28by%20ref%29) (Momentum :type Vector :default "0, 0, 0") (HitInfo :type Hit%20Result%20Structure%20%28by%20ref%29) (InstigatedBy :type Fort%20Pawn%20Object%20Reference) (DamageCauser :type Actor%20Object%20Reference) (EffectContext :type Gameplay%20Effect%20Context%20Handle%20Structure))
  (bind (_bblockinghit _binitialoverlap _time _distance _location _impactpoint _normal _impactnormal) (Collision|BreakHitResult HitInfo))
  (Niagara|SpawnSystematLocation "/Game/Athena/Items/PhysicsActors/BeachBallGiant/FX/NS_Physics_BeachBallGiant_Impact.NS_Physics_BeachBallGiant_Impact" _impactpoint (Math|Conversions|RotationFromXVector (* _impactnormal -1.0))))

(event EventTick
  (params (DeltaSeconds :type Float%20%28single-precision%29)))

(fn BounceAndCount
  (params (Impulse :type Vector) (Location :type Vector) (bVelChange :type Boolean))
  (Utilities|FlowControl|Sequence
    (:then_0
      (Variables|Default|SetlastBounce (Utilities|Time|GetGameTimeinSeconds))
      (if (Networking|IsServer)
        (if (<= (Variables|Default|GetstartedBouncingTime) 0)
          (Variables|Default|SetstartedBouncingTime (Utilities|Time|GetGameTimeinSeconds))
          (else
            _))
        (if (not (Variables|Default|GetisResetting))
          (Variables|Default|SetbounceCountActive true)
          (Variables|Default|SetbounceCount (+ (Variables|Default|GetbounceCount) 1)))))
    (:then_1
      (if (Variables|Default|GetbUseClientPredictionOnImpulse)
        (FortPhysicsSimulationLibrary|AddPhysicsImpulseatLocationwithPrediction self Impulse Location true "None" bVelChange 0 "(MaxSimSizeAllowed=Large,bRejectEmbeddedInLandscape=False)")
        (else
          (FortPhysicsSimulationLibrary|AddPhysicsImpulseatLocation self Impulse Location true "None" bVelChange 0 "(MaxSimSizeAllowed=Large,bRejectEmbeddedInLandscape=False)"))))))

(fn CanBounceNow
  (return (> (Utilities|Time|GetGameTimeinSeconds) (+ (Variables|Default|GetlastBounce) (ScalableFloat|GetValueatLevel (Variables|Default|GetbounceCooldown))))))

(fn GetController
  (return 0))

(fn GetHorizontalDamageImpulse
  (return (ScalableFloat|GetValueatLevel (Variables|Default|GetdamageImpulseHorizontal))))

(fn GetHorizontalHitImpulse
  (return (ScalableFloat|GetValueatLevel (Variables|Default|GethitImpulseHorizontal))))

(fn GetHorizontalPickaxeImpulse
  (return (ScalableFloat|GetValueatLevel (Variables|Default|GetpickaxeImpulseHorizontal))))

(fn GetVerticalDamageImpulse
  (return (ScalableFloat|GetValueatLevel (Variables|Default|GetdamageImpulseVertical))))

(fn GetVerticalHitImpulse
  (return (ScalableFloat|GetValueatLevel (Variables|Default|GethitImpulseVertical))))

(fn GetVerticalPickaxeImpulse
  (return (ScalableFloat|GetValueatLevel (Variables|Default|GetpickaxeImpulseVertical))))

(fn HandleCustomImpulseFromDamage
  (params (ImpulseValue :type Vector :default "0, 0, 0") (Damage :type Float%20%28single-precision%29) (DamageTags :type Gameplay%20Tag%20Container%20Structure%20%28by%20ref%29) (HitResult :type Hit%20Result%20Structure%20%28by%20ref%29) (InstigatingController :type Controller%20Object%20Reference) (DamageCauser :type Actor%20Object%20Reference) (EffectContext :type Gameplay%20Effect%20Context%20Handle%20Structure%20%28by%20ref%29))
  (bind _bhandledimpulse (CallFunction|HandleCustomImpulsefromDamageInternal ImpulseValue Damage DamageTags HitResult InstigatingController DamageCauser EffectContext))
  (return _bhandledimpulse))

(fn HandleCustomImpulseFromDamage_Internal
  (params (ImpulseValue :type Vector) (Damage :type Float%20%28double-precision%29) (DamageTags :type Gameplay%20Tag%20Container%20Structure%20%28by%20ref%29) (HitResult :type Hit%20Result%20Structure%20%28by%20ref%29) (InstigatingController :type Controller%20Object%20Reference) (DamageCauser :type Actor%20Object%20Reference) (EffectContext :type Gameplay%20Effect%20Context%20Handle%20Structure%20%28by%20ref%29))
  (bind (_bblockinghit _binitialoverlap _time _distance _location _impactpoint _normal _impactnormal) (Collision|BreakHitResult HitResult))
  (bind (_returnvalue_x _returnvalue_y) (Math|Vector|BreakVector (* _impactnormal (* (ImpulseGetters/Setters|GetHorizontalDamageImpulse) -1.0))))
  (Utilities|FlowControl|Sequence
    (:then_0
      (if (not (Ability|EffectContext|IsFatalHit EffectContext))
        (if (not (Variables|Default|GetbUseClientPredictionOnImpulse))
          (Networking|SwitchHasAuthority
            (:Authority
              _)))
        (Health|SetHealth (Health|GetMaxHealth))
        (bind _result (CallFunction|CanBounceNow))
        (if _result
          (CallFunction|BounceandCount (Math|Vector|SelectVector (Math|Vector|MakeVector (Utilities|Operators|Multiply (.x (Math|Vector|Normalize (- _impactpoint (Transformation|GetActorLocation (Pawn|GetControlledPawn InstigatingController))))) (Math|Float|Clamp(Float) (- 1.0 (.z (Math|Vector|Normalize (- _impactpoint (Transformation|GetActorLocation (Pawn|GetControlledPawn InstigatingController))))))) (ImpulseGetters/Setters|GetHorizontalPickaxeImpulse)) (* (ImpulseGetters/Setters|GetHorizontalPickaxeImpulse) (.y (Math|Vector|Normalize (- _impactpoint (Transformation|GetActorLocation (Pawn|GetControlledPawn InstigatingController)))))) (* (.z (Math|Vector|Normalize (- _impactpoint (Transformation|GetActorLocation (Pawn|GetControlledPawn InstigatingController))))) (ImpulseGetters/Setters|GetVerticalPickaxeImpulse))) (Math|Vector|MakeVector _returnvalue_x _returnvalue_y (ImpulseGetters/Setters|GetVerticalDamageImpulse)) (GameplayTags|HasAnyTags DamageTags (Variables|Default|GetTC_Pickaxe) true)) (Transformation|GetActorLocation) true)
          (return true))))
    (:then_1
      (return false))))

(fn OnRep_bounceCount)

(fn OnRep_isResetting)

(fn ConstructionScript)
